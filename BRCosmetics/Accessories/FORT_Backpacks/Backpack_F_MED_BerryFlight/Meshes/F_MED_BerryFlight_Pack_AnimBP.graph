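# Animation Blueprint for the BerryFlight backpack that nudges the pack bone while the player skydives.
# Skydive begin and end events set an alpha to 1 or 0, and initialization resets it to 0.
# That alpha drives an additive bone-space Modify Bone offset and roll on the pack, with interpolated alpha smoothing the transition in and out.
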AnimBlueprint /BRCosmetics/Accessories/FORT_Backpacks/Backpack_F_MED_BerryFlight/Meshes/F_MED_BerryFlight_Pack_AnimBP
target_skeleton=&/F_MED_BerryFlight_Pack_Skeleton
graphs [
 AnimGraph {nodes [
   n1 = "Transform (Modify) Bone - Bone: pack"(
    ComponentPose=Local To Component(LocalPose=Input Pose())
    Alpha=Get skydiveAdjustAlpha() -> (skydiveAdjustAlpha:double)
    Translation="0,-1,2"
    Rotation="0,0,15"
   ) {
    rotationMode=BMM_Additive
    rotationSpace=BCS_BoneSpace
    translationMode=BMM_Additive
    translationSpace=BCS_BoneSpace
    alphaScaleBiasClamp {bInterpResult=true, interpSpeedDecreasing=17, interpSpeedIncreasing=17}
    boneToModify {boneName=pack}
    rotation {roll=15}
    translation {y=-1, z=2}
   }
   Output Pose(Result=Component To Local(ComponentPose=@n1)) {layerGroup=None}
  ]}
 EventGraph {nodes [
   Event OnBeginSkydiving()
   > Set skydiveAdjustAlpha(skydiveAdjustAlpha:double=1) -> (Output_Get:double)
   n1 = Event OnEndSkydiving()
   Set skydiveAdjustAlpha(execute=[@n1.then, @n2.then], skydiveAdjustAlpha=0)
   n2 = Event BlueprintInitializeAnimation()
  ]}
]
variables [skydiveAdjustAlpha {type=number}]
